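# Niagara system for the glowstick effects in the Legends Lane player emote.
# Two anim-trail ribbon emitters follow bone/socket pairs: one draws five colored trails, the other two shorter-lived trails, with color selected by spawn group.
# A third emitter bursts two tinted glow sprites pinned through a socket reader.
# Extra offset socket samples position green and purple lights, which fade in with the sprites by emitter age.
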NiagaraSystem /BRCosmetics/Animation/Game/MainPlayer/Emotes/LegendsLane/FX/NS_Emote_LegendsLane_Glowstick
dependencies {
 usedDataInterfaces [
  &NiagaraDataInterfaceCurve
  &NiagaraDataInterfaceSocketReader
  &NiagaraDataInterfaceSkeletalMesh
 ]
 usedDynamicInputs []
 usedModules [
  &self:SystemSpawnScript.SystemScriptSource.SystemScriptGraph.NiagaraNodeAssignment_0.SetVariables_#1
  &SystemState
 ]
}
emitters {
 FN_AnimTrail {
  emitterSpawnScript [&FN_RibbonIDOffsetUtility()]
  emitterUpdateScript [
   &EmitterState()
   &FN_SpawnAnimTrail(Number of Anim Trails=5)
  ]
  particleSpawnScript [
   &FN_InitializeAnimTrailBoneIndecies()
   &FN_InitializeAnimTrailBoneIndecies(AnimTrail Selection=1, First Bone or Socket Index=2, Second Bone or Socket Index=3) {module=FN_InitializeAnimTrailBoneIndecies001}
   &FN_InitializeAnimTrailBoneIndecies(AnimTrail Selection=2, First Bone or Socket Index=4, Second Bone or Socket Index=5) {module=FN_InitializeAnimTrailBoneIndecies002}
   &FN_InitializeAnimTrailBoneIndecies(AnimTrail Selection=3, First Bone or Socket Index=6, Second Bone or Socket Index=7) {module=FN_InitializeAnimTrailBoneIndecies003}
   &FN_InitializeAnimTrailBoneIndecies(AnimTrail Selection=4, First Bone or Socket Index=8, Second Bone or Socket Index=9) {module=FN_InitializeAnimTrailBoneIndecies004}
   &FN_InitializeAnimTrail(
    AnimTrailWidth=&MergeHelper_FloatSwitch(0=1, 1=2.05, 2=1.95, 3=2.05, 4=1, Index=&SpawnGroup(), Num=::NewEnumerator3)
    Color=&MergeHelper_ColorSwitch(
     0=<rgba 1.1816 0.3821 5 1>
     1=<rgba 0.5955 2.1024 5 1>
     2=<rgba 1.1816 0.3821 5 1>
     3=<rgba 5 0.7362 3.5282 1>
     4=<rgba 0.1671 3 1.8483 1>
     Index=&SpawnGroup()
     Num=::NewEnumerator3
    )
    Lifetime=0.275
   )
   &AddVelocity(Velocity=<0 400 0>) {enabled=false}
  ]
  particleUpdateScript [
   &SolveForcesAndVelocity(Clamp Velocity=false, Limit Acceleration=false)
   &FN_UpdateAnimTrail() {module=FN_UpdateAnimTrail001}
   &ScaleColor(
    Scale Alpha=&FloatFromCurve(
     FloatCurve=DataInterface() {propertyValues {Curve {
        defaultValue=1
        keys [{interpMode=RCIM_Cubic, tangentMode=RCTM_User, time=0, value=1}, {interpMode=RCIM_Cubic, tangentMode=RCTM_User, time=1, value=0}]
       }}}
     Scale Curve=0.3
    )
    Scale RGB=&VectorFromFloat(Value=&FloatFromCurve(FloatCurve=DataInterface() {propertyValues {Curve {
         defaultValue=1.8
         keys [
          {tangentMode=RCTM_User, time=0, value=1}
          {arriveTangent=-0.0405, interpMode=RCIM_Cubic, leaveTangent=-0.0405, tangentMode=RCTM_User, time=0.7009, value=1}
          {arriveTangent=-8.0089, interpMode=RCIM_Cubic, leaveTangent=-8.0089, tangentMode=RCTM_User, time=1, value=0}
         ]
        }}}))
    ScaleA=true
    ScaleRGB=true
   )
   &ScaleRibbonWidth(Ribbon Width Scale=&FloatFromCurve(FloatCurve=DataInterface() {propertyValues {Curve {
        defaultValue=1
        keys [{interpMode=RCIM_Cubic, tangentMode=RCTM_User, time=0, value=1}, {interpMode=RCIM_Cubic, tangentMode=RCTM_User, time=1, value=0}]
       }}}))
  ]
  rendererClasses [&NiagaraRibbonRendererProperties]
  renderers [
   {rendererClass=&NiagaraRibbonRendererProperties}
   {rendererClass=&NiagaraRibbonRendererProperties, rendererIndex=1}
  ]
 }
 FN_AnimTrail002 {
  emitterSpawnScript [&FN_RibbonIDOffsetUtility()]
  emitterUpdateScript [
   &EmitterState()
   &FN_SpawnAnimTrail(Number of Anim Trails=2)
  ]
  particleSpawnScript [
   &FN_InitializeAnimTrailBoneIndecies()
   &FN_InitializeAnimTrailBoneIndecies(AnimTrail Selection=1, First Bone or Socket Index=8, Second Bone or Socket Index=9) {module=FN_InitializeAnimTrailBoneIndecies001}
   &FN_InitializeAnimTrailBoneIndecies(AnimTrail Selection=2, First Bone or Socket Index=4, Second Bone or Socket Index=5) {enabled=false, module=FN_InitializeAnimTrailBoneIndecies002}
   &FN_InitializeAnimTrailBoneIndecies(AnimTrail Selection=3, First Bone or Socket Index=6, Second Bone or Socket Index=7) {enabled=false, module=FN_InitializeAnimTrailBoneIndecies003}
   &FN_InitializeAnimTrailBoneIndecies(AnimTrail Selection=4, First Bone or Socket Index=8, Second Bone or Socket Index=9) {enabled=false, module=FN_InitializeAnimTrailBoneIndecies004}
   &FN_InitializeAnimTrail(
    Color=&MergeHelper_ColorSwitch(0=<rgba 2.5775 0 5 1>, 1=<rgba 0 1 0.3061 1>, 2=?, 3=?, Index=&SpawnGroup(), Num=::NewEnumerator0)
    Lifetime=0.15
   )
   &AddVelocity(Velocity=<0 400 0>) {enabled=false}
   &DynamicMaterialParameters(
    Index 0 Param 1=1
    Index 0 Param 2=0.5
    Module.Index0Float0Write=true
    Module.Index0Float1Write=true
    Module.Index0Float2Write=true
    Module.Index0Float3Write=true
   )
  ]
  particleUpdateScript [
   &SolveForcesAndVelocity(Clamp Velocity=false, Limit Acceleration=false)
   &FN_UpdateAnimTrail() {module=FN_UpdateAnimTrail001}
   &ScaleColor(
    Scale Alpha=&FloatFromCurve(
     FloatCurve=DataInterface() {propertyValues {Curve {
        defaultValue=1
        keys [{interpMode=RCIM_Cubic, tangentMode=RCTM_User, time=0, value=1}, {interpMode=RCIM_Cubic, tangentMode=RCTM_User, time=1, value=0}]
       }}}
     Scale Curve=0.2
    )
    Scale RGB=&VectorFromFloat(Value=&FloatFromCurve(FloatCurve=DataInterface() {propertyValues {Curve {
         defaultValue=1.8
         keys [
          {tangentMode=RCTM_User, time=0, value=1}
          {arriveTangent=-0.0405, interpMode=RCIM_Cubic, leaveTangent=-0.0405, tangentMode=RCTM_User, time=0.7009, value=1}
          {arriveTangent=-8.0089, interpMode=RCIM_Cubic, leaveTangent=-8.0089, tangentMode=RCTM_User, time=1, value=0}
         ]
        }}}))
    ScaleA=true
    ScaleRGB=true
   )
   &ScaleRibbonWidth(Ribbon Width Scale=&FloatFromCurve(FloatCurve=DataInterface() {propertyValues {Curve {
        defaultValue=1
        keys [
         {interpMode=RCIM_Cubic, tangentMode=RCTM_User, time=0.8, value=1}
         {arriveTangent=-4.059, interpMode=RCIM_Cubic, leaveTangent=-4.059, tangentMode=RCTM_User, time=1, value=0}
        ]
       }}})) {enabled=false}
  ]
  rendererClasses [&NiagaraRibbonRendererProperties]
  renderers [
   {rendererClass=&NiagaraRibbonRendererProperties}
   {rendererClass=&NiagaraRibbonRendererProperties, rendererIndex=1}
  ]
 }
 Minimal {
  emitterUpdateScript [
   &EmitterState()
   &SpawnBurst_Instantaneous(Use Spawn Probability=false, UseLoopCountLimit=false)
   &SpawnBurst_Instantaneous(Spawn Group=1, Use Spawn Probability=false, UseLoopCountLimit=false) {module=SpawnBurst_Instantaneous001}
  ]
  particleSpawnScript [
   &FN_SetSpawnGroup(Number of Spawn Groups=2) {enabled=false}
   &InitializeParticle(
    Color=&MaskLinearColorBySpawnGroup(Masked Linear Color=<rgba 0 2 0.8636 0.093>, Passthrough Linear Color=<rgba 0.8658 0 2 0.093>)
    Lifetime=5
    Sprite Size Mode=::NewEnumerator3
    Uniform Sprite Size=72
    UsePositionOffset=false
    Write Material Random=true
   )
  ]
  particleUpdateScript [
   &ParticleState()
   &SocketLocation(
    SamplePosOffset=true
    Sampled Position Offset=&MaskVectorBySpawnGroup(Masked Vector=<0 0 15>, Passthrough Vector=<0 0 0>)
    Socket Index=2
    Socket Reader=$System.Socket Reader
    Socket Sampling Mode=::NewEnumerator3
    VelocitySpeedLimit=false
   )
   &ScaleColor(
    Scale Mode=1
    Scale RGBA=&MakeVector4_FromFloat(Float=&FloatFromCurve(
      FloatCurve=DataInterface() {propertyValues {Curve {
         defaultValue=0
         keys [{interpMode=RCIM_Cubic, tangentMode=RCTM_User, time=0, value=0}, {interpMode=RCIM_Cubic, tangentMode=RCTM_User, time=0.45, value=1}]
        }}}
      CurveIndex=$Emitter.Age
     ))
    ScaleRGBA=true
   )
   &SocketLocation(
    Orientation Sampling=::NewEnumerator1
    Position Sampling=::NewEnumerator1
    SamplePosOffset=true
    Sampled Position Offset=<0 0 22>
    Socket Index=2
    Socket Reader=$System.Socket Reader
    Socket Sampling Mode=::NewEnumerator3
    VelocitySpeedLimit=false
   ) {module=SocketLocation001}
   &SocketLocation(
    Orientation Sampling=::NewEnumerator1
    Position Sampling=::NewEnumerator1
    SamplePosOffset=true
    Sampled Position Offset=<0 0 -4>
    Socket Index=2
    Socket Reader=$System.Socket Reader
    Socket Sampling Mode=::NewEnumerator3
    VelocitySpeedLimit=false
   ) {module=SocketLocation002}
   &FN_Light_Attributes(
    Color=<rgba 0.138 1 0 1>
    Create Color Attribute=true
    Create Exponent Attribute=true
    Create Position Attribute=true
    Create Radius Attribute=true
    Exponent=4
    Position:NiagaraPosition=$Output.SocketLocation001.SampledPosition
    Radius=75
    Scale Color=&FloatFromCurve(
     FloatCurve=DataInterface() {propertyValues {Curve {keys [<at 0 0 RCIM_Cubic>,<at 0.45 1 RCIM_Cubic>]}}}
     CurveIndex=$Emitter.Age
     Scale Curve=6
    )
   )
   &FN_Light_Attributes(
    Color=<rgba 0.6092 0 1 1>
    Create Color Attribute=true
    Create Exponent Attribute=true
    Create Position Attribute=true
    Create Radius Attribute=true
    Exponent=4
    Position:NiagaraPosition=$Output.SocketLocation002.SampledPosition
    Radius=75
    Scale Color=&FloatFromCurve(
     FloatCurve=DataInterface() {propertyValues {Curve {keys [<at 0 0 RCIM_Cubic>,<at 0.45 1 RCIM_Cubic>]}}}
     CurveIndex=$Emitter.Age
     Scale Curve=6
    )
   ) {module=FN_Light_Attributes001}
  ]
  rendererClasses [&NiagaraSpriteRendererProperties, &NiagaraLightRendererProperties]
  renderers [
   {rendererClass=&NiagaraSpriteRendererProperties}
   {rendererClass=&NiagaraLightRendererProperties, rendererIndex=1}
   {rendererClass=&NiagaraLightRendererProperties, rendererIndex=2}
  ]
 }
}
summary {
 systemName=NS_Emote_LegendsLane_Glowstick
 emitters [
  {emitterName=FN_AnimTrail, rendererClasses [&NiagaraRibbonRendererProperties]}
  {emitterName=FN_AnimTrail002, rendererClasses [&NiagaraRibbonRendererProperties]}
  {emitterName=Minimal, rendererClasses [&NiagaraSpriteRendererProperties, &NiagaraLightRendererProperties]}
 ]
 userVariables [User.RibbonIDOffset {type=NiagaraInt32}]
}
